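; Anim notify for the SatinCheddar emote that configures the ice blast effect spawned when it fires.
; Received_Notify calls the parent implementation, then fetches the spawned effect and checks it is valid.
; It copies the notify's variables into effect parameters: spike and puddle offsets, plus flags for right hand, first blast and last right-hand blast.
; This lets one effect vary per blast through data set on each notify.
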(fn Received_Notify
  (params (MeshComp :type Skeletal%20Mesh%20Component%20Object%20Reference) (Animation :type Anim%20Sequence%20Base%20Object%20Reference) (EventReference :type Anim%20Notify%20Event%20Reference%20Structure%20%28by%20ref%29))
  (bind _returnvalue (|CallParentFunction MeshComp Animation EventReference))
  (bind _returnvalue_2 (AnimNotify|GetSpawnedEffect))
  (Utilities|IsValid _returnvalue_2
    (:Is%20Valid
      (Effects|Components|ParticleSystem|SetVectorParameter _returnvalue_2 "SpikeOffset" (Variables|Default|GetSpikeOffset))
      (Effects|Components|ParticleSystem|SetBoolParameter _returnvalue_2 "bIsRightHand" (Variables|Default|GetisRightHand))
      (Effects|Components|ParticleSystem|SetBoolParameter _returnvalue_2 "bIsLastRightBlast" (Variables|Default|GetisLastRightBlast))
      (Effects|Components|ParticleSystem|SetVectorParameter _returnvalue_2 "PuddleOffset" (Variables|Default|GetPuddleOffset))
      (Effects|Components|ParticleSystem|SetBoolParameter _returnvalue_2 "bIsFirstBlast" (Variables|Default|GetisFirstBlast))
      (return _returnvalue))))
